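# Unlit additive effects material, named for ray meshes, usable on mesh particles and Niagara sprite, ribbon and mesh renderers.
# The V texture coordinate is cubed into a procedural falloff gradient, tinted by particle color RGB and passed to MF_FX_EmissiveAdjustments as emissive.
# A squared gradient, an inverted gradient and a smooth reverse ramp texture sample are also present but feed nothing.
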Material /Game/Effects/Fort_Effects/Mesh/Rays/M_Rays
graph {
 properties {
  bUsedWithMeshParticles=true
  bUsedWithNiagaraMeshParticles=true
  bUsedWithNiagaraRibbons=true
  bUsedWithNiagaraSprites=true
  blendMode=BLEND_Additive
  pixelDepthOffsetMode=PDOM_Legacy
  shadingModel=MSM_Unlit
  usedShadingModels=Unlit
 }
 nodes [
  n1 = ComponentMask(None=TextureCoordinate()) {g=true}
  MaterialFunctionCall(Emissive=Multiply(A=ParticleColor().RGB, B=Power(Base=@n1) {constExponent=3})) {materialFunction=&MF_FX_EmissiveAdjustments}
  Multiply(A=@n1, B=@n1)
  OneMinus(None=@n1)
  TextureSample() {texture=/Game/Effects/Fort_Effects/Textures/Ramps/T_Ramp_Smooth_Reverse}
 ]
}
